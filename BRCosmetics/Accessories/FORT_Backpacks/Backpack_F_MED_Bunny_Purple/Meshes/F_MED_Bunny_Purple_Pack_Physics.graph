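# Physics asset for the purple bunny backpack cosmetic, defining bodies and joints for its pack and dyn_ bones.
# Every bone gets an identically sized small capsule rather than a fitted shape.
# The constraint tree hangs from root through pack to the head, which carries both ear chains and the upper body, then arms, lower body, feet and tail.
# Joints share tight angular limits, with wider swing only at head-to-pack.
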PhysicsAsset /BRCosmetics/Accessories/FORT_Backpacks/Backpack_F_MED_Bunny_Purple/Meshes/F_MED_Bunny_Purple_Pack_Physics
bodies {
 dyn_arm_L_01 [{shapeType=Capsule, length=2, radius=2, shapeName=dyn_arm_L_01_capsule}]
 dyn_arm_L_02 [{shapeType=Capsule, length=2, radius=2, shapeName=dyn_arm_L_02_capsule}]
 dyn_arm_R_01 [{shapeType=Capsule, length=2, radius=2, shapeName=dyn_arm_R_01_capsule}]
 dyn_arm_R_02 [{shapeType=Capsule, length=2, radius=2, shapeName=dyn_arm_R_02_capsule}]
 dyn_ear_L_01 [{shapeType=Capsule, length=2, radius=2, shapeName=dyn_ear_L_01_capsule}]
 dyn_ear_L_02 [{shapeType=Capsule, length=2, radius=2, shapeName=dyn_ear_L_02_capsule}]
 dyn_ear_L_03 [{shapeType=Capsule, length=2, radius=2, shapeName=dyn_ear_L_03_capsule}]
 dyn_ear_R_01 [{shapeType=Capsule, length=2, radius=2, shapeName=dyn_ear_R_01_capsule}]
 dyn_ear_R_02 [{shapeType=Capsule, length=2, radius=2, shapeName=dyn_ear_R_02_capsule}]
 dyn_ear_R_03 [{shapeType=Capsule, length=2, radius=2, shapeName=dyn_ear_R_03_capsule}]
 dyn_foot_L_01 [{shapeType=Capsule, length=2, radius=2, shapeName=dyn_foot_L_01_capsule}]
 dyn_foot_L_02 [{shapeType=Capsule, length=2, radius=2, shapeName=dyn_foot_L_02_capsule}]
 dyn_foot_R_01 [{shapeType=Capsule, length=2, radius=2, shapeName=dyn_foot_R_01_capsule}]
 dyn_foot_R_02 [{shapeType=Capsule, length=2, radius=2, shapeName=dyn_foot_R_02_capsule}]
 dyn_head [{shapeType=Capsule, length=2, radius=2, shapeName=dyn_head_capsule}]
 dyn_lowerBody [{shapeType=Capsule, length=2, radius=2, shapeName=dyn_lowerBody_capsule}]
 dyn_tail [{shapeType=Capsule, length=2, radius=2, shapeName=dyn_tail_capsule}]
 dyn_upperBody [{shapeType=Capsule, length=2, radius=2, shapeName=dyn_upperBody_capsule}]
 pack [{shapeType=Capsule, length=2, radius=2, shapeName=pack_capsule}]
 root [{shapeType=Capsule, length=2, radius=2, shapeName=root_capsule}]
}
constraints [
 {bone1Name=pack, bone2Name=root, swing1LimitDegrees=10, swing2LimitDegrees=10, twistLimitDegrees=10}
 {bone1Name=dyn_head, bone2Name=pack, swing1LimitDegrees=20, swing2LimitDegrees=20, twistLimitDegrees=10}
 {bone1Name=dyn_ear_L_01, bone2Name=dyn_head, swing1LimitDegrees=10, swing2LimitDegrees=10, twistLimitDegrees=10}
 {bone1Name=dyn_ear_L_02, bone2Name=dyn_ear_L_01, swing1LimitDegrees=10, swing2LimitDegrees=10, twistLimitDegrees=10}
 {bone1Name=dyn_ear_L_03, bone2Name=dyn_ear_L_02, swing1LimitDegrees=10, swing2LimitDegrees=10, twistLimitDegrees=10}
 {bone1Name=dyn_ear_R_01, bone2Name=dyn_head, swing1LimitDegrees=10, swing2LimitDegrees=10, twistLimitDegrees=10}
 {bone1Name=dyn_ear_R_02, bone2Name=dyn_ear_R_01, swing1LimitDegrees=10, swing2LimitDegrees=10, twistLimitDegrees=10}
 {bone1Name=dyn_ear_R_03, bone2Name=dyn_ear_R_02, swing1LimitDegrees=10, swing2LimitDegrees=10, twistLimitDegrees=10}
 {bone1Name=dyn_upperBody, bone2Name=dyn_head, swing1LimitDegrees=10, swing2LimitDegrees=10, twistLimitDegrees=10}
 {bone1Name=dyn_arm_L_01, bone2Name=dyn_upperBody, swing1LimitDegrees=10, swing2LimitDegrees=10, twistLimitDegrees=10}
 {bone1Name=dyn_arm_L_02, bone2Name=dyn_arm_L_01, swing1LimitDegrees=10, swing2LimitDegrees=10, twistLimitDegrees=10}
 {bone1Name=dyn_arm_R_01, bone2Name=dyn_upperBody, swing1LimitDegrees=10, swing2LimitDegrees=10, twistLimitDegrees=10}
 {bone1Name=dyn_arm_R_02, bone2Name=dyn_arm_R_01, swing1LimitDegrees=10, swing2LimitDegrees=10, twistLimitDegrees=10}
 {bone1Name=dyn_lowerBody, bone2Name=dyn_upperBody, swing1LimitDegrees=10, swing2LimitDegrees=10, twistLimitDegrees=10}
 {bone1Name=dyn_foot_R_01, bone2Name=dyn_lowerBody, swing1LimitDegrees=10, swing2LimitDegrees=10, twistLimitDegrees=10}
 {bone1Name=dyn_foot_R_02, bone2Name=dyn_foot_R_01, swing1LimitDegrees=10, swing2LimitDegrees=10, twistLimitDegrees=10}
 {bone1Name=dyn_foot_L_01, bone2Name=dyn_lowerBody, swing1LimitDegrees=10, swing2LimitDegrees=10, twistLimitDegrees=10}
 {bone1Name=dyn_foot_L_02, bone2Name=dyn_foot_L_01, swing1LimitDegrees=10, swing2LimitDegrees=10, twistLimitDegrees=10}
 {bone1Name=dyn_tail, bone2Name=dyn_lowerBody, swing1LimitDegrees=10, swing2LimitDegrees=10, twistLimitDegrees=10}
]
